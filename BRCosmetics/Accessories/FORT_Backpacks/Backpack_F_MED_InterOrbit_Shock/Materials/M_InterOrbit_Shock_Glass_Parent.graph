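# Translucent parent material for the glass of the InterOrbit Shock backpack, with an animated smoky glow inside.
# MF_Glass feeds a masked thin-film function, then optional camera dither and a vertex-hide switch between position-offset collapse and character opacity.
# Inner smoke layers two panning, crystal-distorted textures on triplanar UVs with BumpOffset parallax, adding a tinted emissive-mask glow.
# Inverse Fresnel and a glass mask control the emissive blend before TechArt overrides.
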Material /BRCosmetics/Accessories/FORT_Backpacks/Backpack_F_MED_InterOrbit_Shock/Materials/M_InterOrbit_Shock_Glass_Parent
graph {
 properties {
  bEnableResponsiveAA=true
  bScreenSpaceReflections=true
  bUseMaterialAttributes=true
  bUsedWithClothing=true
  bUsedWithMeshDeformer=true
  bUsedWithMorphTargets=true
  bUsedWithNanite=true
  bUsedWithNiagaraMeshParticles=true
  bUsedWithSkeletalMesh=true
  blendMode=BLEND_Translucent
  ditherOpacityMask=true
  pixelDepthOffsetMode=PDOM_Legacy
  translucencyDirectionalLightingIntensity=5
  translucencyLightingMode=TLM_SurfacePerPixelLighting
  usedShadingModels=Default Lit
 }
 nodes [
  n1 = TextureSampleParameter2D() {
   group=00_Glass
   parameterName=TechArtMask
   samplerType=SAMPLERTYPE_Masks
   sortPriority=-96
   texture=&T_BlackMask
  }
  n2 = MaterialFunctionCall!(
   Material Attributes=MaterialFunctionCall!(TechArtMask=@n1.RGB) {materialFunction=/Game/Global/Materials/Character_Parent_2023/Glass_Parent_2023/MF_Glass}.Result
   Fresnel=Fresnel() {baseReflectFraction=0, exponent=1}
   Alpha=@n1.RGB
  ) {materialFunction=/Game/Global/MaterialFunctions/TechArt/MF_ThinFilm_TechArt}
  n3 = StaticSwitchParameter(
   True=MaterialFunctionCall!(In=@n2.Result) {materialFunction=/Game/Packages/Fortress_SharedMaterials/Base_Material_Functions/BaseCustomizationFunctions/MF_FN_Vehicle_CameraDither}.Result
   False=@n2.Result
  ) {group=00_Glass, parameterName=UseCameraDither, sortPriority=250}
  n4 = Multiply(A=Subtract(A=ObjectPositionWS(), B=WorldPosition().XYZ), B=ScalarParameter() {group=00_Glass, parameterName=VertexHide, sortPriority=301})
  n5 = GetMaterialAttributes(None=StaticSwitchParameter(
    True=SetMaterialAttributes!(MaterialAttributes=@n3, World Position Offset=@n4) {attributeSetTypes [#1]}
    False=MaterialFunctionCall!(Material=@n3) {materialFunction=/Game/Global/Materials/Character_Parent_2023/Functions/BaseFN_CharacterOpacity_2023}.Result
   ) {group=00_Glass, parameterName=UseVertexHide, sortPriority=300}) {attributeGetTypes [#2]}
  n6 = NamedRerouteUsage()
  n7 = ScalarParameter() {defaultValue=1.1002, parameterName=Param_7}
  n8 = TextureSample() {
   samplerType=SAMPLERTYPE_Masks
   texture=/BRCosmetics/Accessories/FORT_Backpacks/Backpack_F_MED_InterOrbit_Shock/Textures/T_InterOrbit_Shock_Backpack_TAMask
  }
  n9 = NamedRerouteDeclaration(None=@n8.G) {name=GlassMask, nodeColor=<rgba 0 0.0859 1 1>}
  n10 = LinearInterpolate(A=OneMinus(None=Fresnel(ExponentIn=@n7)), B=@n9, Alpha=@n9)
  MaterialFunctionCall(In=SetMaterialAttributes!(MaterialAttributes=@n5.MaterialAttributes, Emissive Color=LinearInterpolate(A=@n5.Emissive Color, B=@n6, Alpha=@n10)) {attributeSetTypes [#2]}) {materialFunction=&MF_TechArtOverrides}
  n11 = TextureSample() {
   samplerType=SAMPLERTYPE_Masks
   texture=/BRCosmetics/Accessories/FORT_Backpacks/Backpack_F_MED_InterOrbit_Shock/Textures/T_InterOrbit_Shock_Backpack_E2
  }
  n12 = Multiply(A=OneMinus(None=@n11.R), B=OneMinus(None=NamedRerouteUsage()))
  n13 = VectorParameter() {defaultValue=<rgba 20 0.519 0.4 1>}
  n14 = MaterialFunctionCall!(
   TriplanarUV_UseClamp?=Constant()
   TriplanarUV_FlipUV=Constant()
   TriplanarUV_CenterPosition=Constant4Vector() {constant=<rgba 0.9523 0 15.2308 1>}.RGB
   TriplanarUV_UVScale=ScalarParameter() {defaultValue=35, parameterName=Param_6}
  ) {materialFunction=/BRCosmetics/Characters/Player/Male/Medium/Bodies/M_MED_BirchPalace/Skins/Gas/Materials/MF_TriplanarUV}
  n15 = Multiply(A=@n14.UV_FrontProjection, B=ScalarParameter() {defaultValue=4})
  n16 = Constant() {r=1}
  n17 = Multiply(A=Subtract(A=Time(), B=Multiply(A=@n16) {constB=2}), B=Constant() {r=2})
  n18 = VectorParameter() {parameterName=InsideSmokeSpeed}
  n19 = Panner(Coordinate=@n15, Time=@n17, Speed=ComponentMask(None=@n18.RGBA) {g=true, r=true}) {speedX=-0.04, speedY=0.25}
  n20 = TextureSample(UVs=Multiply(A=@n14.UV_FrontProjection, B=ScalarParameter() {defaultValue=2, parameterName=Param_3}) {constB=8}) {
   samplerSource=SSM_Wrap_WorldGroupSettings
   samplerType=SAMPLERTYPE_LinearGrayscale
   texture=/Game/Global/Textures/Default/Masks_and_Gradients/T_MF_Crystal2
  }
  n21 = Multiply(A=@n20.R, B=ScalarParameter() {defaultValue=0.05, group=Crack, parameterName=InsideSmokeIntensity})
  n22 = ScalarParameter() {defaultValue=-2, parameterName=Param_4}
  n23 = TextureSample(UVs=BumpOffset(Coordinate=Add(A=@n19, B=@n21), Height=@n22)) {texture=&T_Inky_Smoke_Tile_Chromatic}
  n24 = Panner(Coordinate=@n15, Time=@n17, Speed=ComponentMask(None=@n18.RGBA) {a=true, b=true}) {speedX=0.05, speedY=0.05}
  n25 = Multiply(A=@n20.R) {constB=0.1}
  n26 = TextureSample(UVs=BumpOffset(Coordinate=Add(A=@n24, B=@n25), Height=@n22)) {texture=&T_Inky_Smoke_Tile_Chromatic}
  n27 = Multiply(
   A=SphereMask(A=Add(A=@n23.R, B=@n26.R), B=Constant(), Radius=Constant() {r=0.25}, Hardness=ScalarParameter() {defaultValue=0.3, parameterName=Param_1}) {hardnessPercent=1}
   B=Constant3Vector() {constant=<rgba 0.5 0.0231 0.0202 1>}
  )
  n28 = ScalarParameter() {defaultValue=0.75, parameterName=Param_2}
  NamedRerouteDeclaration(None=Add(A=Multiply(A=@n12, B=@n13.RGB), B=Multiply(A=@n27, B=@n28))) {name=Inner_Smoke, nodeColor=<rgba 0 0.2266 1 1>}
  Power(Base=@n11.R)
  ScalarParameter() {defaultValue=-2, parameterName=Param_5}
  TextureCoordinate() {uTiling=3, vTiling=1.4}
  TextureCoordinate() {uTiling=1.5, vTiling=0.7}
 ]
}
