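# Parent material for the InterOrbit Shock backpack, adding crystal shading, animated inner smoke and masked glow to the 2023 character material stack.
# MF_Crystal follows the PreFX function, then Base Color and Emissive are modified before PostFX.
# Smoke sums two panning inky-smoke samples, one bump-offset for parallax, and thresholds them with a SphereMask.
# A packed mask texture supplies smoke and edge masks, while a UV sphere mask locally dims emissive.
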Material /BRCosmetics/Accessories/FORT_Backpacks/Backpack_F_MED_InterOrbit_Shock/Materials/M_InterOrbit_Shock_Parent
graph {
 properties {
  bAutomaticallySetUsageInEditor=false
  bUseMaterialAttributes=true
  bUsedWithMeshDeformer=true
  bUsedWithMorphTargets=true
  bUsedWithSkeletalMesh=true
  pixelDepthOffsetMode=PDOM_Legacy
  shadingModel=MSM_FromMaterialExpression
  usedShadingModels=Default Lit | Subsurface | Clear Coat | Subsurface Profile | Cloth
 }
 nodes [
  n1 = MaterialFunctionCall() {materialFunction=&MF_CharacterParent_2023_PreFX}
  n2 = GetMaterialAttributes(None=MaterialFunctionCall!(Material=@n1.Result, Alpha=NamedRerouteUsage()) {materialFunction=/Game/Global/MaterialFunctions/TechArt/MF_Crystal}.Result) {attributeGetTypes [#1, #2]}
  n3 = NamedRerouteUsage()
  n4 = Add(A=Multiply(A=@n3, B=Constant() {r=3}), B=NamedRerouteUsage())
  n5 = NamedRerouteUsage()
  n6 = NamedRerouteUsage()
  n7 = NamedRerouteUsage()
  n8 = LinearInterpolate(A=@n6, B=Add(A=@n6, B=NamedRerouteUsage()), Alpha=@n7) {constB=0}
  n9 = Multiply(A=@n8, B=Constant() {r=0.2})
  n10 = TextureCoordinate()
  n11 = AppendVector(A=Constant() {r=0.406}, B=Constant() {r=0.408})
  n12 = Constant() {r=0.15}
  n13 = Constant() {r=0.416}
  n14 = Multiply(A=@n7, B=Power(Base=SphereMask(A=@n10, B=@n11, Radius=@n12, Hardness=@n13)))
  MaterialFunctionCall(
   Material=SetMaterialAttributes!(
    MaterialAttributes=@n2.MaterialAttributes
    Base Color=LinearInterpolate(A=@n3, B=@n4, Alpha=@n5)
    Emissive Color=LinearInterpolate(A=@n8, B=@n9, Alpha=@n14)
   ) {attributeSetTypes [#1, #2]}
   M=@n1.M
   Pre Skinned Local Position=@n1.M
   Distance Blend=@n1.M
   Fresnel=@n1.M
  ) {materialFunction=&MF_CharacterParent_2023_PostFX}
  n15 = TextureSample() {
   samplerType=SAMPLERTYPE_Masks
   texture=/BRCosmetics/Accessories/FORT_Backpacks/Backpack_F_MED_InterOrbit_Shock/Textures/T_InterOrbit_Shock_Backpack_TAMask
  }
  NamedRerouteDeclaration(None=ComponentMask(None=@n15.R) {g=true}) {name=SmokeMask, nodeColor=<rgba 0 0.9062 1 1>}
  NamedRerouteDeclaration(None=@n2.Base Color) {name=In_BaseColor, nodeColor=<rgba 1 0.7734 0 1>}
  n16 = Constant() {r=8}
  NamedRerouteDeclaration(None=LinearInterpolate(A=@n2.Emissive Color, B=Multiply(A=@n2.Emissive Color, B=@n16), Alpha=NamedRerouteUsage())) {name=In_EmissiveColor, nodeColor=<rgba 1 0.6562 0 1>}
  n17 = Multiply(A=TextureCoordinate() {uTiling=1.5, vTiling=0.7}, B=Constant() {r=4})
  n18 = Constant() {r=1}
  n19 = Multiply(A=Subtract(A=Time(), B=Multiply(A=@n18) {constB=2}), B=Constant() {r=2})
  n20 = VectorParameter() {parameterName=InsideSmokeSpeed, defaultValue=<rgba 0 0.04 0 0.08>}
  n21 = ComponentMask(None=@n20.RGBA) {g=true, r=true}
  n22 = Panner(Coordinate=@n17, Time=@n19, Speed=ComponentMask(None=@n20.RGBA) {a=true, b=true}) {speedX=0.05, speedY=0.1}
  n23 = Constant() {r=-2}
  n24 = Add(
   A=TextureSample(UVs=Panner(Coordinate=@n17, Time=@n19, Speed=@n21) {speedX=-0.04, speedY=0.5}) {texture=&T_Inky_Smoke_Tile_Chromatic}.R
   B=TextureSample(UVs=BumpOffset(Coordinate=@n22, Height=@n23)) {texture=&T_Inky_Smoke_Tile_Chromatic}.R
  )
  n25 = Constant()
  n26 = Constant() {r=0.25}
  n27 = Constant() {r=0.15}
  n28 = NamedRerouteUsage()
  n29 = Constant() {r=-0.5}
  NamedRerouteDeclaration(None=Multiply(A=SphereMask(A=@n24, B=@n25, Radius=@n26, Hardness=@n27) {hardnessPercent=1}, B=Multiply(A=@n28, B=@n29))) {name=Inner_Smoke, nodeColor=<rgba 0 0.2266 1 1>}
  NamedRerouteDeclaration(None=@n15.B) {name=EdgeMask, nodeColor=<rgba 1 0 0.6328 1>}
 ]
}
