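# Parent material for the InterOrbit Shock backpack's Riftshard, giving it a crystal look with smoky interior and glowing edges.
# Character PreFX output runs through MF_Crystal, then has base color, emissive, normal, roughness and other attributes overridden before PostFX.
# Inner smoke pans two textures over triplanar front-projection UVs, one distorted by a crystal depth texture, then tints and Fresnel-blends them.
# Emissive adds Fresnel-boosted glow and a colored Fresnel rim, scaled globally.
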Material /BRCosmetics/Accessories/FORT_Backpacks/Backpack_F_MED_InterOrbit_Shock/Materials/M_InterOrbit_Shock_Riftshard_parent
graph {
 properties {
  bUseMaterialAttributes=true
  bUsedWithMeshDeformer=true
  bUsedWithNanite=true
  bUsedWithSkeletalMesh=true
  pixelDepthOffsetMode=PDOM_Legacy
  shadingModel=MSM_FromMaterialExpression
  usedShadingModels=Default Lit | Subsurface | Clear Coat | Subsurface Profile | Cloth
 }
 nodes [
  n1 = MaterialFunctionCall() {materialFunction=&MF_CharacterParent_2023_PreFX}
  n2 = GetMaterialAttributes(None=MaterialFunctionCall!(Material=@n1.Result, Alpha=Constant() {r=1}) {materialFunction=/Game/Global/MaterialFunctions/TechArt/MF_Crystal}.Result) {attributeGetTypes [
    #1
    #2
    #3
    #4
    #5
    #6
   ]}
  n3 = SetMaterialAttributes!(
   MaterialAttributes=@n2.MaterialAttributes
   Roughness=Saturate(None=NamedRerouteUsage())
   Metallic=ScalarParameter() {defaultValue=0.1, group=Fresnel, parameterName=Metallic_Offset}
   Specular=ScalarParameter() {defaultValue=1, group=Fresnel, parameterName=Specular_Offset}
   Refraction=Constant() {r=1.76}
   Normal=MaterialFunctionCall(Normal=NamedRerouteUsage(), Flatness=Constant() {r=-0.05}) {materialFunction=&FlattenNormal}.Result
   Base Color=NamedRerouteUsage()
   Emissive Color=NamedRerouteUsage()
   World Position Offset
  ) {attributeSetTypes [
    #1
    #2
    #3
    #7
    #4
    #5
    #6
    #8
   ]}
  MaterialFunctionCall(Material=@n3, M=@n1.M, Pre Skinned Local Position=@n1.M, Distance Blend=@n1.M, Fresnel=@n1.M) {materialFunction=&MF_CharacterParent_2023_PostFX}
  NamedRerouteDeclaration(None=@n2.Base Color) {name=In_BaseColor, nodeColor=<rgba 1 0 0.4453 1>}
  NamedRerouteDeclaration(None=@n2.Emissive Color) {name=In_EmissiveColor, nodeColor=<rgba 0 0.7656 1 1>}
  n4 = TextureSampleParameter2D() {
   group=SheerWiill
   parameterName=SheerWill_E
   texture=/BRCosmetics/Accessories/FORT_Backpacks/Backpack_F_MED_InterOrbit_Shock/Textures/T_InterOrbit_Shock_Backpack_E
  }
  NamedRerouteDeclaration(None=@n4.RGB) {name=SheerWill_E_RGB, nodeColor=<rgba 0 1 0.5781 1>}
  NamedRerouteDeclaration(None=@n2.Roughness) {name=In_Roughness, nodeColor=<rgba 0 1 0.2734 1>}
  n5 = MaterialFunctionCall!(
   TriplanarUV_UseClamp?=Constant()
   TriplanarUV_FlipUV=Constant()
   TriplanarUV_CenterPosition=Constant4Vector() {constant=<rgba 0.9523 0 15.2308 1>}.RGB
   TriplanarUV_UVScale=Constant() {r=20}
  ) {materialFunction=/BRCosmetics/Characters/Player/Male/Medium/Bodies/M_MED_BirchPalace/Skins/Gas/Materials/MF_TriplanarUV}
  n6 = Multiply(A=@n5.UV_FrontProjection, B=ScalarParameter() {defaultValue=2, group=SheerWiill, parameterName=SmokeTiling})
  n7 = Constant() {r=1}
  n8 = Multiply(A=Subtract(A=Time(), B=Multiply(A=@n7) {constB=2}), B=Constant() {r=2})
  n9 = VectorParameter() {parameterName=InsideSmokeSpeed}
  n10 = ComponentMask(None=@n9.RGBA) {g=true, r=true}
  n11 = TextureSample(UVs=Multiply(A=@n5.UV_FrontProjection, B=ScalarParameter() {defaultValue=2, group=SheerWiill, parameterName=DepthTiling}) {constB=2}) {
   samplerSource=SSM_Wrap_WorldGroupSettings
   samplerType=SAMPLERTYPE_LinearGrayscale
   texture=/Game/Global/Textures/Default/Masks_and_Gradients/T_MF_Crystal2
  }
  n12 = ScalarParameter() {defaultValue=0.2, group=Crack, parameterName=InsideSmokeIntensity}
  n13 = TextureSample(UVs=Add(A=Panner(Coordinate=@n6, Time=@n8, Speed=@n10) {speedX=-0.04, speedY=1}, B=Multiply(A=@n11.R, B=@n12))) {texture=&T_Inky_Smoke_Tile_Chromatic}
  n14 = TextureSample(UVs=Panner(Coordinate=@n6, Time=@n8, Speed=ComponentMask(None=@n9.RGBA) {a=true, b=true}) {speedX=0.05, speedY=0.5}) {texture=&T_Inky_Smoke_Tile_Chromatic}
  n15 = Multiply(
   A=SphereMask(A=Add(A=@n13.R, B=@n14.R), B=Constant(), Radius=Constant() {r=0.25}, Hardness=Constant() {r=0.5}) {hardnessPercent=1}
   B=Multiply(A=VectorParameter() {group=Crack, parameterName=InsideSmokeColor, defaultValue=<rgba 0.88 0.0001 0.0295 1>}.RGB) {constB=1.5}
  )
  n16 = ScalarParameter() {defaultValue=0.3891, group=Crack, parameterName=InsideSmokeFresnelExponentln}
  n17 = Saturate(None=Multiply(
    A=Fresnel(ExponentIn=@n16) {baseReflectFraction=0, exponent=3}
    B=ScalarParameter() {defaultValue=1.5, group=Crack, parameterName=Crack_DesaturationAmount}
   ) {constB=2})
  NamedRerouteDeclaration(None=LinearInterpolate(B=LinearInterpolate(A=@n15, Alpha=@n17), Alpha=ScalarParameter() {defaultValue=0.35, group=SheerWiill, parameterName=Smoke_Blend}) {constA=1}) {name=Inner_Smoke, nodeColor=<rgba 0 0.2266 1 1>}
  NamedRerouteDeclaration(None=@n2.Normal) {name=In_Normal, nodeColor=<rgba 0 0.4609 1 1>}
  n18 = Multiply(A=NamedRerouteUsage(), B=ScalarParameter() {defaultValue=4, group=SheerWiill, parameterName=BaseColor_Strength})
  n19 = NamedRerouteUsage()
  NamedRerouteDeclaration(None=Desaturation(None=Multiply(A=@n18, B=@n19), Fraction=ScalarParameter() {defaultValue=0.05, group=SheerWiill, parameterName=BaseColor_Desaturation})) {name=OutBaseColor, nodeColor=<rgba 0.7812 1 0 1>}
  n20 = Multiply(
   A=ScalarParameter() {defaultValue=2, group=SheerWiill, parameterName=FresnelEdgeIntensity}
   B=VectorParameter() {group=SheerWiill, parameterName=FresnelEdgeColor, defaultValue=<rgba 1 0.5 0.5 1>}.RGB
  )
  n21 = Power(Base=Fresnel(ExponentIn=ScalarParameter() {defaultValue=5, group=SheerWiill, parameterName=FresnelEdge_Exponentln}))
  n22 = Multiply(A=NamedRerouteUsage(), B=ScalarParameter() {defaultValue=10, group=Emissive, parameterName=Emissive_Strength})
  n23 = ScalarParameter() {defaultValue=1, group=Emissive, parameterName=Emissive_FresnelExponentln}
  n24 = LinearInterpolate(B=ScalarParameter() {defaultValue=2.2114, group=Emissive, parameterName=Emissive_FresnelPower}, Alpha=Power(Base=Fresnel(ExponentIn=@n23))) {constB=2}
  n25 = Add(A=Multiply(A=NamedRerouteUsage(), B=Constant() {r=0.24}), B=Add(A=Multiply(A=@n22, B=@n24), B=Multiply(A=@n20, B=@n21)))
  n26 = NamedRerouteUsage()
  NamedRerouteDeclaration(None=Multiply(A=Multiply(A=@n25, B=@n26), B=ScalarParameter() {defaultValue=5, group=SheerWiill, parameterName=Emissive_Global})) {name=OutEmissiveColor, nodeColor=<rgba 0.8984 0 1 1>}
 ]
}
