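# Animation Blueprint for the InterOrbit Shock backpack cosmetic, adjusting its pack bone and four floating crystal bones.
# The AnimGraph chains additive Transform Bone nodes on the input pose, including a fixed pack offset and tilt.
# The update event computes per-crystal bob vectors from game-time sine waves at differing rates.
# It also maps spine_05-to-spine_01 yaw delta from ref pose into pack rotation and translation variables, and stores skydiving state as a float.
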AnimBlueprint /BRCosmetics/Accessories/FORT_Backpacks/Backpack_F_MED_InterOrbit_Shock/Meshes/F_MED_InterOrbit_Shock_Pack_AnimBP
target_skeleton=&/F_MED_InterOrbit_Shock_Pack_Skeleton
graphs [
 AnimGraph {nodes [
   n1 = Input Pose()
   n2 = "Transform (Modify) Bone - Bone: pack"(
    ComponentPose="Transform (Modify) Bone - Bone: pack"(ComponentPose=Local To Component(LocalPose=@n1), Translation="0,0,0", Rotation="0,0,0") {rotationMode=BMM_Additive, translationMode=BMM_Additive, boneToModify {boneName=pack}}
    Translation="0,-2,0"
    Rotation="0,0,-2"
   ) {
    rotationMode=BMM_Additive
    rotationSpace=BCS_BoneSpace
    translationMode=BMM_Additive
    translationSpace=BCS_BoneSpace
    boneToModify {boneName=pack}
    rotation {roll=-2}
    translation {y=-2}
   }
   n3 = "Transform (Modify) Bone - Bone: dyn_crystal_03"(
    ComponentPose="Transform (Modify) Bone - Bone: dyn_crystal_02"(
     ComponentPose="Transform (Modify) Bone - Bone: dyn_crystal_01"(ComponentPose=@n2, Alpha=1, Translation="0,0,0", Rotation="0,0,0", Scale="1,1,1") {translationMode=BMM_Additive, translationSpace=BCS_BoneSpace, boneToModify {boneName=dyn_crystal_01}}
     Alpha=1
     Translation="0,0,0"
     Rotation="0,0,0"
     Scale="1,1,1"
    ) {translationMode=BMM_Additive, translationSpace=BCS_BoneSpace, boneToModify {boneName=dyn_crystal_02}}
    Alpha=1
    Translation="0,0,0"
    Rotation="0,0,0"
    Scale="1,1,1"
   ) {translationMode=BMM_Additive, translationSpace=BCS_BoneSpace, boneToModify {boneName=dyn_crystal_03}}
   Output Pose(Result=Component To Local(ComponentPose="Transform (Modify) Bone - Bone: dyn_crystal_04"(ComponentPose=@n3, Alpha=1, Translation="0,0,0", Rotation="0,0,0", Scale="1,1,1") {translationMode=BMM_Additive, translationSpace=BCS_BoneSpace, boneToModify {boneName=dyn_crystal_04}})) {layerGroup=None}
  ]}
 EventGraph {nodes [
   Event BlueprintUpdateAnimation()
   > n1 = Sequence()
   n2 = GetGameTimeInSeconds()
   n3 = "Sin (Radians)"(A=float * float(A=float * float(A=0.9, B=0.9), B=@n2))
   n4 = "Sin (Radians)"(A=float * float(A=float * float(A=0.8, B=0.8), B=@n2))
   n5 = "Sin (Radians)"(A=float * float(A=float * float(A=0.75, B=0.75), B=@n2))
   > Set crystal_1_float(crystal_1_float:Vector=MakeVector(X=float * float(A=@n5, B=0.7), Y=float * float(A=@n3, B=0.8), Z=float * float(A=@n4, B=0.7))) -> (Output_Get:Vector)
   n6 = "Sin (Radians)"(A=float * float(A=float * float(A=0.65, B=0.65), B=@n2))
   n7 = "Sin (Radians)"(A=float * float(A=float * float(A=0.75, B=0.75), B=@n2))
   > Set crystal_2_float(crystal_2_float:Vector=MakeVector(X=float * float(A=@n7, B=0.6), Y=0, Z=float * float(A=@n6, B=0.8))) -> (Output_Get:Vector)
   n8 = "Sin (Radians)"(A=float * float(A=float * float(A=1.2, B=1.2), B=@n2))
   n9 = "Sin (Radians)"(A=float * float(A=float * float(A=0.7, B=0.7), B=@n2))
   > Set crystal_3_float(crystal_3_float:Vector=MakeVector(X=float * float(A=@n9, B=0.5), Y=0, Z=float * float(A=@n8, B=0.4))) -> (Output_Get:Vector)
   n10 = "Sin (Radians)"(A=float * float(A=float * float(A=0.35, B=0.35), B=@n2))
   n11 = "Sin (Radians)"(A=float * float(A=float * float(A=0.75, B=0.75), B=@n2))
   n12 = "Sin (Radians)"(A=float * float(A=float * float(A=0.25, B=0.25), B=@n2))
   > Set crystal_4_float(crystal_4_float:Vector=MakeVector(X=float * float(A=@n11, B=0.8), Y=float * float(A=@n10, B=0.65), Z=float * float(A=@n12, B=0.7))) -> (Output_Get:Vector)
   n13 = GetOwnerSkeletalMeshComponent()
   n14 = BreakRotator(InRot=BreakTransform(InTransform=GetDeltaTransformFromRefPose(self=@n13, BoneName=spine_05, BaseName=spine_01)).Rotation)
   Set backpack_rotate(execute=@n1.then_1, backpack_rotate=MakeRotator(Roll=MapRangeClamped(Value=@n14.Yaw, InRangeA=0, InRangeB=-20, OutRangeB=5)))
   > Set backpack_translation(backpack_translation=MakeVector(Y=MapRangeClamped(Value=@n14.Yaw, InRangeA=0, InRangeB=-30, OutRangeA=-0.5)))
   > Set Skydiving(Skydiving:double="To Float (Boolean)"(InBool=Get bIsSkyDiving())) -> (Output_Get:double)
   MapRangeClamped(Value=0, InRangeA=0, InRangeB=-30, OutRangeB=30)
   n15 = float * float(A=0.7, B=7)
   float * float(A="Sin (Radians)"(A=float * float(A=@n15, B=@n2)), B=0.75)
   n16 = float * float(A=0.8, B=0.9)
   float * float(A="Sin (Radians)"(A=float * float(A=@n16, B=@n2)), B=0.4)
  ]}
]
variables [
 backpack_rotate {type=Rotator}
 backpack_translation {type=Vector}
 crystal_1_float {type=Vector}
 crystal_2_float {type=Vector}
 crystal_3_float {type=Vector}
 crystal_4_float {type=Vector}
 Skydiving {type=number}
]
